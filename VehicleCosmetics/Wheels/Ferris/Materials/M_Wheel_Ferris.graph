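# Masked, material-attributes material for the Ferris vehicle wheel cosmetic, usable on skeletal meshes and with mesh deformers.
# MF_WheelParent_Simple supplies the base attributes and a rim RGBA mask.
# A static switch sets Opacity Mask to either the inverted rim mask alpha, cutting out parts of the wheel, or constant one.
# The result then passes through the shared vehicle damage and de-rez transition functions.
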Material /VehicleCosmetics/Wheels/Ferris/Materials/M_Wheel_Ferris
graph {
 properties {
  bAutomaticallySetUsageInEditor=false
  bUseMaterialAttributes=true
  bUsedWithMeshDeformer=true
  bUsedWithSkeletalMesh=true
  blendMode=BLEND_Masked
  pixelDepthOffsetMode=PDOM_Legacy
  previewMesh=/VehicleCosmetics/Wheels/Ferris/Meshes/SK_Wheel_Ferris
  usedShadingModels=Default Lit
 }
 nodes [
  n1 = MaterialFunctionCall() {materialFunction=&MF_WheelParent_Simple}
  n2 = StaticSwitchParameter(True=OneMinus(None=ComponentMask(None=@n1.RimRGBAMask) {a=true}), False=Constant() {r=1}) {parameterName=Opacity mask}
  MaterialFunctionCall(In=MaterialFunctionCall(Material=SetMaterialAttributes!(MaterialAttributes=@n1.MaterialAttributes, Opacity Mask=@n2) {attributeSetTypes [#1]}) {materialFunction=&MF_BaseVehicle_Damage}.Result) {materialFunction=&MF_VehicleDeRez_Transition}
 ]
}
